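# Physics asset for the female Space Wanderer backpack, giving collision and joint setup to its meteor piece.
# A sphere body on the dyn_meteor bone, offset along local X, is joined by a single constraint to a small box body on the pack bone.
# The constraint locks twist, so any relative rotation of the meteor against the pack comes from swing.
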PhysicsAsset /Game/Accessories/FORT_Backpacks/Backpack_F_MED_Space_Wanderer/Meshes/F_MED_Space_Wanderer_Physics
bodies {dyn_meteor [{shapeType=Sphere, radius=10.8834, center=<3.8871 0 0>}], pack [{shapeType=Box, extentX=0.75, extentY=0.75, extentZ=0.75}]}
constraints [{bone1Name=dyn_meteor, bone2Name=pack, swing1LimitDegrees=45, swing2LimitDegrees=45, twistLimitDegrees=45, twistMotion=Locked}]
