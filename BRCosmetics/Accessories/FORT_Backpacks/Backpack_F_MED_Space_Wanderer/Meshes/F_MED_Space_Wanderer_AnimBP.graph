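# Animation Blueprint for the Space Wanderer backpack cosmetic, animating its dyn_meteor bone.
# The incoming pose passes through a bone-space rotation replace driven by SpinnerRot, then an additive translation offset weighted by JumpAlpha with interpolated scale and bias.
# A RigidBody physics node is present but unconnected to the output chain.
# The EventGraph is empty, and neither driving variable is declared here.
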AnimBlueprint /BRCosmetics/Accessories/FORT_Backpacks/Backpack_F_MED_Space_Wanderer/Meshes/F_MED_Space_Wanderer_AnimBP
target_skeleton=&/F_MED_Space_Wanderer_Skeleton
graphs [
 AnimGraph {nodes [
   RigidBody!(ComponentPose, Alpha) -> (Pose) {
    bClampLinearTranslationLimitToRefPose=true
    bOverrideWorldGravity=true
    overridePhysicsAsset=/Game/Accessories/FORT_Backpacks/Backpack_F_MED_Space_Wanderer/Meshes/F_MED_Space_Wanderer_Physics
    simulationSpace=BaseBoneSpace
    alphaBoolBlend {blendOption=Linear}
    baseBoneRef {boneName=root}
    componentAppliedLinearAccClamp=<3000 3000 3000>
    componentLinearAccScale=<0.25 0.25 0.25>
    componentLinearVelScale=<0.25 0.25 0.25>
    externalForce=<0.25 0.25 0.25>
    overrideWorldGravity {z=-800}
   }
   n1 = "Transform (Modify) Bone - Bone: dyn_meteor"(
    ComponentPose=Local To Component(LocalPose=Input Pose())
    Alpha=1
    Translation="0,0,0"
    Rotation=Get SpinnerRot() -> (SpinnerRot:Rotator)
    Scale="1,1,1"
   ) {rotationMode=BMM_Replace, rotationSpace=BCS_BoneSpace, alphaBoolBlend {blendOption=Linear}, boneToModify {boneName=dyn_meteor}}
   n2 = "Transform (Modify) Bone - Bone: dyn_meteor"(ComponentPose=@n1, Alpha=Get JumpAlpha() -> (JumpAlpha:float), Translation="-8,0,0", Rotation="0,0,0", Scale="1,1,1") {
    translationMode=BMM_Additive
    translationSpace=BCS_BoneSpace
    alphaBoolBlend {blendOption=Linear}
    alphaScaleBiasClamp {bInterpResult=true, bias=0.15, interpSpeedDecreasing=4, interpSpeedIncreasing=5, scale=0.75}
    boneToModify {boneName=dyn_meteor}
    translation {x=-8}
   }
   Output Pose(Result=Component To Local(ComponentPose=@n2)) {layerGroup=None}
  ]}
 EventGraph {links [], nodes []}
]
variables []
